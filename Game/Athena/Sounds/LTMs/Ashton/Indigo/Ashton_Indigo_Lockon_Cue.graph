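# Sound cue for the lock-on target indicator in the Ashton Indigo limited-time mode.
# A single wave player feeds the root directly, with no randomization or mixing.
# It uses gadget-activate attenuation and the non-critical in-game UI sound class.
# A concurrency override allows one instance at a time, stopping the oldest when a new one starts.
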SoundCue /Game/Athena/Sounds/LTMs/Ashton/Indigo/Ashton_Indigo_Lockon_Cue
root=WavePlayer_0
nodes [{id=WavePlayer_0, properties {soundWaveAssetPtr=&/Ashton_Indigo_Target_Indicator_01}}]
properties {
 attenuationSettings=&Fort_GadgetActivate_Att
 bOverrideConcurrency=true
 maxDistance=4000
 soundClassObject=&Fort_UI_InGame_NonCritical
 volumeMultiplier=1
 concurrencyOverrides {maxCount=1, resolutionRule=StopOldest}
}
