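# Sound cue for the whoosh of an Ashton Indigo LTM projectile passing by.
# A Random node picks one of six equally weighted whoosh-by waves, then a Modulator varies the result, with a raised upper pitch limit.
# The cue uses the ranged weapon sound class and bullet whiz-by concurrency rules, and its attenuation override turns off volume falloff while setting a wide stereo spread.
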SoundCue /Game/Athena/Sounds/LTMs/Ashton/Indigo/Ashton_Indigo_Proj_By_Cue
root=Modulator_0
nodes [
 {id=Modulator_0, children [Random_0], properties {pitchMax=1.15}}
 {id=Random_0, children [WavePlayer_0,WavePlayer_1,WavePlayer_2,WavePlayer_3,WavePlayer_4,WavePlayer_6], properties {weights [1,1,1,1,1,1]}}
 {id=WavePlayer_0, properties {soundWaveAssetPtr=&/Ashton_Indigo_Primary_Whoosh_By_01}}
 {id=WavePlayer_1, properties {soundWaveAssetPtr=&/Ashton_Indigo_Primary_Whoosh_By_02}}
 {id=WavePlayer_2, properties {soundWaveAssetPtr=&/Ashton_Indigo_Primary_Whoosh_By_03}}
 {id=WavePlayer_3, properties {soundWaveAssetPtr=&/Ashton_Indigo_Primary_Whoosh_By_04}}
 {id=WavePlayer_4, properties {soundWaveAssetPtr=&/Ashton_Indigo_Primary_Whoosh_By_05}}
 {id=WavePlayer_6, properties {soundWaveAssetPtr=&/Ashton_Indigo_Primary_Whoosh_By_06}}
]
properties {
 bOverrideAttenuation=true
 maxDistance=2097152
 soundClassObject=&Fort_WeaponRanged
 volumeMultiplier=0.6
 attenuationOverrides {bAttenuate=false, stereoSpread=5000}
 concurrencySet [/Game/Sounds/Concurrency_Rules/Athena_BulletWhizBy_Conc]
}
