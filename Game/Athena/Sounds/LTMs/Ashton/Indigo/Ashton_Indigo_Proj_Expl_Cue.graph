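# Sound cue for the Ashton Indigo projectile explosion, built as three distance-layered variants mixed together.
# Close, medium and distant layers each pick one of four random waves, get pitch variation and use their own attenuation; the distant layer reuses explosive bow arrow explosion sounds.
# The close layer starts after a brief delay, varies volume and overrides its falloff distance.
# Concurrency is capped at two instances, stopping the oldest.
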SoundCue /Game/Athena/Sounds/LTMs/Ashton/Indigo/Ashton_Indigo_Proj_Expl_Cue
root=Mixer_0
nodes [
 {
  id=Attenuation_0
  children [Modulator_2]
  properties {attenuationSettings=/Game/Sounds/Attenuation/Fort_Weapon_Milo_Medium_Explosion_Athena_Att}
 }
 {
  id=Attenuation_1
  children [Modulator_0]
  properties {attenuationSettings=/Game/Sounds/Attenuation/Fort_Weapon_Milo_Distant_Explosion_Athena_Att}
 }
 {
  id=Attenuation_7
  children [Modulator_6]
  properties {attenuationSettings=/Game/Sounds/Attenuation/Fort_Weapon_Milo_Close_Explosion_Athena_Att, attenuationOverrides {falloffDistance=1600}}
 }
 {id=Delay_0, children [Random_0], properties {delayMax=0.066, delayMin=0.066}}
 {id=Mixer_0, children [Attenuation_7, Attenuation_0, Attenuation_1], properties {inputVolume [1,0.7,0.7]}}
 {id=Modulator_0, children [Random_6], properties {pitchMax=1.1, pitchMin=0.9}}
 {id=Modulator_2, children [Random_3], properties {pitchMax=1.1, pitchMin=0.9}}
 {id=Modulator_6, children [Delay_0], properties {pitchMax=1.3, pitchMin=1, volumeMin=0.5}}
 {id=Random_0, children [WavePlayer_0,WavePlayer_1,WavePlayer_2,WavePlayer_3], properties {weights [1,1,1,1]}}
 {id=Random_3, children [WavePlayer_14, WavePlayer_15, WavePlayer_16, WavePlayer_17], properties {weights [1,1,1,1]}}
 {id=Random_6, children [WavePlayer_18, WavePlayer_19, WavePlayer_20, WavePlayer_21], properties {weights [1,1,1,1]}}
 {id=WavePlayer_0, properties {soundWaveAssetPtr=&/Ashton_Indigo_Primary_Expl_Close_01}}
 {id=WavePlayer_1, properties {soundWaveAssetPtr=&/Ashton_Indigo_Primary_Expl_Close_02}}
 {id=WavePlayer_14, properties {soundWaveAssetPtr=&/Ashton_Indigo_Primary_Expl_Med_01}}
 {id=WavePlayer_15, properties {soundWaveAssetPtr=&/Ashton_Indigo_Primary_Expl_Med_02}}
 {id=WavePlayer_16, properties {soundWaveAssetPtr=&/Ashton_Indigo_Primary_Expl_Med_03}}
 {id=WavePlayer_17, properties {soundWaveAssetPtr=&/Ashton_Indigo_Primary_Expl_Med_04}}
 {id=WavePlayer_18, properties {soundWaveAssetPtr=/Game/Sounds/Fort_Weapons/Explosive_Bow/Weapon_Bow_Compound_Arrow_Expl_Dist_01}}
 {id=WavePlayer_19, properties {soundWaveAssetPtr=/Game/Sounds/Fort_Weapons/Explosive_Bow/Weapon_Bow_Compound_Arrow_Expl_Dist_02}}
 {id=WavePlayer_2, properties {soundWaveAssetPtr=&/Ashton_Indigo_Primary_Expl_Close_03}}
 {id=WavePlayer_20, properties {soundWaveAssetPtr=/Game/Sounds/Fort_Weapons/Explosive_Bow/Weapon_Bow_Compound_Arrow_Expl_Dist_03}}
 {id=WavePlayer_21, properties {soundWaveAssetPtr=/Game/Sounds/Fort_Weapons/Explosive_Bow/Weapon_Bow_Compound_Arrow_Expl_Dist_04}}
 {id=WavePlayer_3, properties {soundWaveAssetPtr=&/Ashton_Indigo_Primary_Expl_Close_04}}
]
properties {
 bOverrideConcurrency=true
 maxDistance=15000
 soundClassObject=&Fort_WeaponRanged_Fire_3P_PSM
 volumeMultiplier=0.7
 concurrencyOverrides {maxCount=2, resolutionRule=StopOldest}
}
